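# Conversation database for the DerangedMile NPC (Miles Cross), giving his greetings and dialogue when players talk to him.
# The opening speech picks contextual lines by requirement, priority and weight: first-ever meeting, per-match conversation count checks, and reactions to specific player outfits.
# It branches to linked quest, basic-services and Season 39 PerfectGift conversations, plus an "Ask About" menu with a multi-step Peelians story that returns to the start.
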ConversationDatabase /NPCLibrary/NPCs/DerangedMile/NPCLibrary_Conversation_DerangedMile
entry_points [{tag=Conversation.Entry.DerangedMile, destinations [#1]}]
nodes [
 ConversationEntryPointNode_7 {
  id=#1
  class=ConversationEntryPointNode
  outputs [#2]
  properties {entryTag {tagName=Conversation.Entry.DerangedMile}}
 }
 Speech_C_4 {
  id=#2
  class=Speech_C
  outputs [#3, #4, #5, #6]
  properties {
   nodeName=""
   generalConfig {
    defaultMessage="Yeah, it's me, Miles Cross. Want an autograph?"
    contextualMessages [
     {
      message="Yeah, it's me, Miles Cross. Want an autograph?"
      contextRequirements [{requirement {class=FortControllerRequirement_IsFirstEverConversationWithNPC}}]
      priority {curve {rowName=Tandem.Speech.Priority.FirstEverInteraction}}
     }
     {
      message="Some actors are real snobs about explosions. I love 'em!"
      contextRequirements [{requirement {
         class=FortControllerRequirement_CheckNPCMatchConversationCount
         properties {nPCUniqueIdTag {tagName=AISpawnerData.Type.Specialist.Heavy.DerangedMile}}
        }}]
      priority {curve {curveTable=null, rowName=None}}
     }
     {
      message=There's nothing better than a popcorn movie with a love story and a bunch of fight scenes.
      contextRequirements [{requirement {
         class=FortControllerRequirement_CheckNPCMatchConversationCount
         properties {nPCUniqueIdTag {tagName=AISpawnerData.Type.Specialist.Heavy.DerangedMile}}
        }}]
      priority {curve {curveTable=null, rowName=None}}
     }
     {
      message=I don't do any of this for awards. I do this for the fans.
      contextRequirements [{requirement {
         class=FortControllerRequirement_CheckNPCMatchConversationCount
         properties {nPCUniqueIdTag {tagName=AISpawnerData.Type.Specialist.Heavy.DerangedMile}}
        }}]
      priority {curve {curveTable=null, rowName=None}}
     }
     {
      message=Dark times comin'. I liked it better when the dark times were just in the movies...
      contextRequirements [{requirement {
         class=FortControllerRequirement_CheckNPCMatchConversationCount
         properties {nPCUniqueIdTag {tagName=AISpawnerData.Type.Specialist.Heavy.DerangedMile}}
        }}]
      priority {curve {curveTable=null, rowName=None}}
      weight {value=2}
     }
     {
      message="Careful out there, Cat. I think some REAL danger's on the horizon..."
      contextRequirements [{requirement {class=FortControllerRequirement_HasCID, properties {allowedCIds [/BRCosmetics/Athena/Items/Cosmetics/Characters/Character_RowLiaison_Cafe]}}}]
      priority {curve {rowName=Tandem.Speech.Priority.OutfitReaction}}
     }
     {
      message="Let's hope Peelians isn't our last movie, eh buddy?"
      contextRequirements [{requirement {class=FortControllerRequirement_HasCID, properties {allowedCIds [/BRCosmetics/Athena/Items/Cosmetics/Characters/Character_FloralMane_Sand]}}}]
      priority {curve {rowName=Tandem.Speech.Priority.OutfitReaction}}
     }
     {
      message="Man, it's an honor to finally work with you."
      contextRequirements [{requirement {class=FortControllerRequirement_HasCID, properties {allowedCIds [/BRCosmetics/Athena/Items/Cosmetics/Characters/Character_ZenCrash]}}}]
      priority {curve {rowName=Tandem.Speech.Priority.OutfitReaction}}
     }
     {
      message="Can't wait to introduce you to the world, man. They're gonna love you."
      contextRequirements [{requirement {
         class=FortControllerRequirement_HasCID
         properties {allowedCIds [
           &Character_BananaPhilosopher
           &Character_BananaAdventure
           /BRCosmetics/Athena/Items/Cosmetics/Characters/Character_BananaCake
           &CID_349_Athena_Commando_M_Banana
           &CID_701_Athena_Commando_M_BananaAgent
           &CID_814_Athena_Commando_M_BananaSummer
           /BRCosmetics/Athena/Items/Cosmetics/Characters/CID_A_045_Athena_Commando_M_BananaLeader
           &CID_A_323_Athena_Commando_M_BananaWinter
           /BRCosmetics/Athena/Items/Cosmetics/Characters/CID_TBD_Athena_Commando_M_Banana_CINE
          ]}
        }}]
      priority {curve {rowName=Tandem.Speech.Priority.OutfitReaction}}
     }
    ]
   }
  }
 }
 ConversationLinkNode_2 {
  id=#6
  class=ConversationLinkNode
  properties {remoteEntryTag {tagName=Conversation.Entry.Season39.WF.PerfectGift}}
 }
 Speech_C_2 {
  id=#5
  class=Speech_C
  outputs [#7, #8]
  properties {nodeName="", softTaskIcon=&T_UI_NPC_question, generalConfig {defaultMessage=Hit me.}}
  sub_nodes [ConversationChoiceNode_0 {class=ConversationChoiceNode, properties {defaultChoiceDisplayText=Ask About...}}]
 }
 Back_C_2 {
  id=#8
  class=Back_C
  sub_nodes [ConversationChoiceNode_6 {class=ConversationChoiceNode, properties {defaultChoiceDisplayText=BACK}}]
 }
 Speech_C_11 {
  id=#7
  class=Speech_C
  outputs [#9]
  properties {
   nodeName=""
   softTaskIcon=&T_UI_NPC_question
   generalConfig {defaultMessage=Peelians was gonna be my perfect ninety-minute popcorn flick.}
  }
  sub_nodes [ConversationChoiceNode_1 {class=ConversationChoiceNode, properties {defaultChoiceDisplayText=Peelians}}]
 }
 Speech_C_3 {
  id=#9
  class=Speech_C
  outputs [#10]
  properties {
   nodeName=""
   softTaskIcon=&T_UI_NPC_question
   generalConfig {defaultMessage=Legendary director. Forte-winning leading lady. Stunt king. We had it all.}
  }
  sub_nodes [ConversationChoiceNode_3 {class=ConversationChoiceNode, properties {defaultChoiceDisplayText=Continue...}}]
 }
 Speech_C_5 {
  id=#10
  class=Speech_C
  outputs [#11]
  properties {
   nodeName=""
   softTaskIcon=&T_UI_NPC_question
   generalConfig {defaultMessage=Maybe it'll still come out someday... but I've got a bad feeling.}
  }
  sub_nodes [ConversationChoiceNode_4 {class=ConversationChoiceNode, properties {defaultChoiceDisplayText=Continue...}}]
 }
 Speech_C_7 {
  id=#11
  class=Speech_C
  outputs [#12]
  properties {
   nodeName=""
   softTaskIcon=&T_UI_NPC_question
   generalConfig {defaultMessage=I think dark times are comin' to this Island.}
  }
  sub_nodes [ConversationChoiceNode_5 {class=ConversationChoiceNode, properties {defaultChoiceDisplayText=Continue...}}]
 }
 ReturnToStart_C_23 {
  id=#12
  class=ReturnToStart_C
  properties {softTaskIcon=&T_UI_NPC_Back}
  sub_nodes [ConversationChoiceNode_2 {class=ConversationChoiceNode, properties {defaultChoiceDisplayText=BACK}}]
 }
 ConversationLinkNode_0 {id=#4, class=ConversationLinkNode, properties {remoteEntryTag {tagName=Conversation.Entry.BasicServices}}}
 ConversationLinkNode_1 {id=#3, class=ConversationLinkNode, properties {remoteEntryTag {tagName=Conversation.Entry.NPCLibrary.Quests}}}
]
